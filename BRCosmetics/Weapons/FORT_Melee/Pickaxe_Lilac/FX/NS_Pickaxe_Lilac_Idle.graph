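# Niagara idle effect for the Lilac pickaxe that plays an occasional sheen sweep followed by a glint.
# Both emitters burst with System.SpawnSheen as spawn probability; system-level assignments and a User.bisSwinging input exist, but their wiring is not shown.
# Sheen_Mesh animates a dynamic material parameter while fading color in and out.
# Sheen_Glint follows 0.45 seconds later with a short-lived, offset sprite that pops and shrinks.
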NiagaraSystem /BRCosmetics/Weapons/FORT_Melee/Pickaxe_Lilac/FX/NS_Pickaxe_Lilac_Idle
dependencies {
 usedDataInterfaces [&NiagaraDataInterfaceCurve]
 usedDynamicInputs [
  &Vector2DFromFloat
  &UniformRangedFloat
  &MakeCustomFloatFromBool
  &SetBoolByIntComparison
  /Niagara/DynamicInputs/Subtract/Subtract_Integer
 ]
 usedModules [
  &self:SystemSpawnScript.SystemScriptSource.SystemScriptGraph.NiagaraNodeAssignment_4.SetVariables_#1
  &SystemState
  &self:SystemSpawnScript.SystemScriptSource.SystemScriptGraph.NiagaraNodeAssignment_3.SetVariables_#2
 ]
}
emitters {
 Sheen_Glint {
  emitterUpdateScript [
   &EmitterState()
   &SpawnBurst_Instantaneous(Spawn Probability=$System.SpawnSheen, Spawn Time=0.45, Use Spawn Probability=true, UseLoopCountLimit=false)
  ]
  particleSpawnScript [&InitializeParticle(
    Color=<rgba 0.5267 2 1.5617 1>
    Lifetime Max=0.33
    Lifetime Min=0.25
    Lifetime Mode=::NewEnumerator1
    Mesh Scale Mode=::NewEnumerator3
    Position Offset=<1 6.2 95.5>
    Sprite Rotation Angle=-15
    Sprite Rotation Mode=::NewEnumerator0
    Sprite Size Max=<80 40>
    Sprite Size Min=<60 30>
    Sprite Size Mode=::NewEnumerator2
    UsePositionOffset=true
    Write Material Random=true
   )]
  particleUpdateScript [
   &ParticleState()
   &ScaleSpriteSize(
    Uniform Curve Scale=&RandomRangeFloat(Minimum=0.9, Maximum=1)
    Uniform Curve Sprite Scale=DataInterface() {propertyValues {Curve {keys [
        {arriveTangent=9.0725, interpMode=RCIM_Cubic, leaveTangent=9.0725, tangentMode=RCTM_User, time=0, value=0}
        {arriveTangent=-2.5249, interpMode=RCIM_Cubic, leaveTangent=-2.5249, tangentMode=RCTM_User, time=0.1408, value=0.9847}
        {arriveTangent=-0.0405, interpMode=RCIM_Cubic, leaveTangent=-0.0405, tangentMode=RCTM_User, time=1, value=0}
       ]}}}
   )
   &CameraOffset(Camera Offset Amount=5)
   &ScaleColor(ScaleA=true, ScaleRGB=true)
   &SpriteRotationRate(Rotation Rate=-90) {enabled=false}
  ]
  rendererClasses [&NiagaraSpriteRendererProperties]
  renderers [{rendererClass=&NiagaraSpriteRendererProperties}]
 }
 Sheen_Mesh {
  emitterUpdateScript [
   &EmitterState()
   &SpawnBurst_Instantaneous(Spawn Probability=$System.SpawnSheen, Use Spawn Probability=true, UseLoopCountLimit=false) {module=SpawnBurst_Instantaneous001}
  ]
  particleSpawnScript [&InitializeParticle(Color=<rgba 1.3635 4 3.8348 1>, Lifetime=0.42, UsePositionOffset=false, Write Material Random=true) {module=InitializeParticle001}]
  particleUpdateScript [
   &ParticleState()
   &DynamicMaterialParameters(
    Index 0 Param 1=&FloatFromCurve(FloatCurve=DataInterface() {propertyValues {Curve {keys [
         {arriveTangent=2.9873, interpMode=RCIM_Cubic, leaveTangent=2.9873, tangentMode=RCTM_User, time=0, value=0}
         {arriveTangent=1.2358, interpMode=RCIM_Cubic, leaveTangent=1.2358, tangentMode=RCTM_User, time=1.1, value=1.8}
        ]}}})
    Module.Param0WriteEnabled=true
    Module.Param1WriteEnabled=true
    Module.Param2WriteEnabled=false
    Module.Param3WriteEnabled=false
   ) {module=DynamicMaterialParameters001}
   &CameraOffset(Camera Offset Amount=0.05)
   &SolveForcesAndVelocity(Clamp Velocity=false, Limit Acceleration=false)
   &GenerateDeathEvent(Use Event Probability=false)
   &ScaleColor(
    Scale Alpha=&FloatFromCurve(FloatCurve=DataInterface() {propertyValues {Curve {keys [<at 0 0>,<at 0.5016 0.9767>,<at 1 0>]}}})
    Scale RGB=&VectorFromFloat(Value=&FloatFromCurve(FloatCurve=DataInterface() {propertyValues {Curve {keys [<at 0 0>,<at 0.5017 1.0042>,<at 1 0>]}}}))
    ScaleA=true
    ScaleRGB=true
   )
  ]
  rendererClasses [&NiagaraMeshRendererProperties]
  renderers [{rendererClass=&NiagaraMeshRendererProperties, rendererIndex=0}]
 }
}
summary {
 systemName=NS_Pickaxe_Lilac_Idle
 emitters [
  {emitterName=Sheen_Glint, rendererClasses [&NiagaraSpriteRendererProperties]}
  {emitterName=Sheen_Mesh, rendererClasses [&NiagaraMeshRendererProperties]}
 ]
 userVariables [User.bisSwinging {type=NiagaraBool}]
}
